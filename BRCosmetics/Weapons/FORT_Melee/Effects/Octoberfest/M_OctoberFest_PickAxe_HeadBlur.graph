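# Unlit translucent two-sided mesh-particle material for the motion blur of the Octoberfest pickaxe head.
# Four panning caustics samples blend and multiply with a turbulence-distorted, rotated smoke trail and U gradient.
# That mask mixes hue-shifted and original particle color for emissive, and shapes alpha alongside a sphere mask positioned by dynamic Erode.
# Dynamic curl parameters also drive vector noise on actor-relative position, replaced by a constant on mobile.
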Material /BRCosmetics/Weapons/FORT_Melee/Effects/Octoberfest/M_OctoberFest_PickAxe_HeadBlur
graph {
 properties {
  bUsedWithMeshParticles=true
  blendMode=BLEND_Translucent
  pixelDepthOffsetMode=PDOM_Legacy
  previewMesh=/Game/Effects/Fort_Effects/Mesh/PickAxe/SM_Octoberfest_Blur_Head
  shadingModel=MSM_Unlit
  translucencyPass=MTP_BeforeDOF
  twoSided=true
  usedShadingModels=Unlit
 }
 nodes [
  n1 = ScalarParameter() {defaultValue=0.02, parameterName="Hue Shift Percentage (S)"}
  n2 = ParticleColor()
  n3 = Panner(Coordinate=TextureCoordinate() {uTiling=0.6, vTiling=0.25}) {speedX=-1, speedY=-0.5}
  n4 = Panner(Coordinate=TextureCoordinate() {uTiling=2, vTiling=0.45}) {speedX=-2, speedY=1}
  n5 = Panner(Coordinate=TextureCoordinate() {uTiling=2, vTiling=0.5}) {speedX=-1, speedY=-0.5}
  n6 = Panner(Coordinate=TextureCoordinate() {uTiling=2, vTiling=0.25}) {speedX=-2, speedY=1}
  n7 = LinearInterpolate(
   A=LinearInterpolate(
    A=TextureSample(UVs=@n3) {texture=&T_ShieldCaustics}.G
    B=TextureSample(UVs=@n4) {texture=&T_ShieldCaustics}.G
   )
   B=LinearInterpolate(
    A=TextureSample(UVs=@n5) {texture=&T_ShieldCaustics}.G
    B=TextureSample(UVs=@n6) {texture=&T_ShieldCaustics}.G
   )
  )
  n8 = TextureSample(UVs=Panner(Coordinate=TextureCoordinate() {uTiling=0.1}) {speedY=0.5}) {texture=/Game/Global/Textures/Default/Caustics/T_Turbulence_Seamless}
  n9 = MaterialFunctionCall(
   UVs=Add(A=TextureCoordinate(), B=Multiply(A=@n8.R) {constB=0.25})
   "Rotation Angle (0-1)"=ScalarParameter() {defaultValue=0.25, parameterName="Rotation Angle (0-1) (S)"}
  ) {materialFunction=&CustomRotator}
  n10 = Multiply(
   A=Multiply(
    A=Power(Base=@n7)
    B=TextureSample(UVs=@n9.Rotated Values) {samplerType=SAMPLERTYPE_Grayscale, texture=&T_Smoke_Trail_Soft}.G
   )
   B=ComponentMask(None=TextureCoordinate()) {r=true}
  )
  MaterialFunctionCall(Emissive=LinearInterpolate(
    A=MaterialFunctionCall(Hue Shift Percentage=@n1, Texture=@n2.RGB) {materialFunction=&HueShift}.Result
    B=@n2.RGB
    Alpha=Multiply(A=@n10, B=@n10)
   )) {materialFunction=&MF_FX_EmissiveAdjustments}
  n11 = DynamicParameter() {defaultValue=<rgba 1 50 15 0>, paramNames [Curl_Speed, Curl_Scale, Curl_Intensity, Erode]}
  n12 = ScalarParameter() {parameterName=B_1}
  Multiply(
   A=Multiply(A=@n2.A, B=@n10)
   B=SphereMask(
    A=TextureCoordinate() {vTiling=0}
    B=AppendVector(A=@n11.Erode, B=@n12)
    Radius=ScalarParameter() {defaultValue=0.5, parameterName=Radius_1}
    Hardness=ScalarParameter() {parameterName=Hardness_1}
   )
  )
  n13 = WorldPosition()
  n14 = ActorPositionWS()
  n15 = DynamicParameter() {defaultValue=<rgba 1 50 15 0>, paramNames [Curl_Speed, Curl_Scale, Curl_Intensity, Erode]}
  n16 = Divide(
   A=Subtract(
    A=Subtract(A=@n13.XYZ, B=@n14)
    B=DynamicParameter() {defaultValue=<rgba 1 50 15 0>, paramNames [Curl_Speed, Curl_Scale, Curl_Intensity, Erode]}.Curl_Speed
   )
   B=@n15.Curl_Scale
  )
  ShadingPathSwitch(
   Default=Multiply(
    A=DynamicParameter() {defaultValue=<rgba 1 50 15 0>, paramNames [Curl_Speed, Curl_Scale, Curl_Intensity, Erode]}.Curl_Intensity
    B=VectorNoise(World Position=@n16) {noiseFunction=VNF_VectorALU}
   )
   Mobile=Constant() {r=1}
  ) {desc=Migrated from FeatureLevelSwitch}
  SphereMask(
   A=TextureCoordinate() {vTiling=0}
   B=AppendVector(A=@n15.Erode, B=ScalarParameter() {parameterName=B})
   Radius=ScalarParameter() {defaultValue=0.5, parameterName=Radius}
   Hardness=ScalarParameter() {parameterName=Hardness}
  )
  VectorParameter() {defaultValue=<rgba 0.7135 0.2468 0.2044 1>}
 ]
}
